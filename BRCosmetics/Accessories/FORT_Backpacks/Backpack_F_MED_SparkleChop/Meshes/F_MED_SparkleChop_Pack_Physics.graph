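# Physics asset for the SparkleChop backpack cosmetic, simulating what its bone names suggest is a zipper pull with a spring piece.
# A box body on the base joint anchors a chain of two small capsules, running from base to zip to spring joint.
# The zip swings within a narrow cone on the base with twist locked.
# The spring is tighter, hinging on the zip in one limited swing axis only.
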PhysicsAsset /BRCosmetics/Accessories/FORT_Backpacks/Backpack_F_MED_SparkleChop/Meshes/F_MED_SparkleChop_Pack_Physics
bodies {
 dyn_base_joint [{
   shapeType=Box
   extentX=3.5
   extentY=6.5
   extentZ=7.5
   shapeName=dyn_base_joint_box
   center=<-6.3414 -7.1367 -10.5084>
   rotation=<rot -10 0 -0.0002>
  }]
 dyn_spring_joint [{shapeType=Capsule, length=5, radius=0.5, shapeName=dyn_spring_joint_capsule, center=<3.5117 0 0>, rotation=<rot -90 179.9978 -179.9971>}]
 dyn_zip_joint [{shapeType=Capsule, length=0.5, radius=0.5, shapeName=dyn_zip_joint_capsule, center=<0.9474 0 0>, rotation=<rot -90 -166.7373 166.7379>}]
}
constraints [
 {bone1Name=dyn_zip_joint, bone2Name=dyn_base_joint, swing1LimitDegrees=15, swing2LimitDegrees=15, twistLimitDegrees=0, twistMotion=Locked}
 {
  bone1Name=dyn_spring_joint
  bone2Name=dyn_zip_joint
  swing1LimitDegrees=0
  swing1Motion=Locked
  swing2LimitDegrees=15
  twistLimitDegrees=0
  twistMotion=Locked
 }
]
