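# Animation Blueprint for the SparkleChop cosmetic backpack, adding physics-driven secondary motion to its skeleton.
# The input pose gets a small additive bone-space offset on dyn_base_joint, then feeds a RigidBody node using the pack's physics asset.
# Simulation runs relative to that joint, with overridden gravity and scaled, clamped component motion to temper the response.
# The event graph holds only unconnected stub nodes, so all behavior lives in the AnimGraph.
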AnimBlueprint /BRCosmetics/Accessories/FORT_Backpacks/Backpack_F_MED_SparkleChop/Meshes/F_MED_SparkleChop_Pack_AnimBP
target_skeleton=&/F_MED_SparkleChop_Pack_Skeleton
graphs [
 AnimGraph {nodes [
   n1 = "Transform (Modify) Bone - Bone: dyn_base_joint"(ComponentPose=Local To Component(LocalPose=Input Pose()), Alpha=1, Translation="0.000,3.391,-0.000", Rotation="0,0,0", Scale="1,1,1") {
    translationMode=BMM_Additive
    translationSpace=BCS_BoneSpace
    alphaBoolBlend {blendOption=Linear}
    boneToModify {boneName=dyn_base_joint}
    translation {y=3.391}
   }
   n2 = RigidBody(ComponentPose=@n1) {
    bClampLinearTranslationLimitToRefPose=true
    bOverrideWorldGravity=true
    overridePhysicsAsset=&/F_MED_SparkleChop_Pack_Physics
    simulationSpace=BaseBoneSpace
    alphaBoolBlend {blendOption=Linear}
    baseBoneRef {boneName=dyn_base_joint}
    componentAppliedLinearAccClamp=<3000 3000 3000>
    componentLinearAccScale=<0.4 0.3 0.3>
    componentLinearVelScale=<0.6 0.6 0.6>
    overrideWorldGravity {z=-980}
    simSpaceSettings {maxAngularAcceleration=2000, maxAngularVelocity=2000, maxLinearAcceleration=2000, maxLinearVelocity=2000, velocityScaleZ=0.8, worldAlpha=0.1}
   }
   Output Pose(Result=Component To Local(ComponentPose=@n2)) {layerGroup=None}
  ]}
 EventGraph {nodes [
   TryGetPawnOwner()
   Event BlueprintUpdateAnimation()
  ]}
]
variables []
